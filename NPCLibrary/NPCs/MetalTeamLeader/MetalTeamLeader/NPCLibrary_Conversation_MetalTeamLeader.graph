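# Conversation database for the Metal Team Leader NPC, giving its greeting and access to shared dialogue services.
# The entry point leads to one speech node with a default line plus contextual lines for first-ever conversations and specific player character cosmetics.
# The first-meeting line takes its priority from a curve row, while the cosmetic lines use zero with no curve.
# The speech node then links to the shared NPC Library quests and basic services conversations.
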ConversationDatabase /NPCLibrary/NPCs/MetalTeamLeader/MetalTeamLeader/NPCLibrary_Conversation_MetalTeamLeader
entry_points [{tag=Conversation.Entry.MetalTeamLeader, destinations [#1]}]
nodes [
 ConversationEntryPointNode_7 {
  id=#1
  class=ConversationEntryPointNode
  outputs [#2]
  properties {entryTag {tagName=Conversation.Entry.MetalTeamLeader}}
 }
 Speech_C_4 {
  id=#2
  class=Speech_C
  outputs [#3, #4]
  properties {
   nodeName=""
   generalConfig {
    defaultMessage=Beginning relaxation protocol. Please ignore the screeching beeps.
    contextualMessages [
     {
      message="What is this \"peace,\" flesh-friend?"
      contextRequirements [{requirement {class=FortControllerRequirement_IsFirstEverConversationWithNPC}}]
      priority {curve {rowName=Tandem.Speech.Priority.FirstEverInteraction}}
     }
     {
      message="I read about your destiny on your Llamapedia page. "
      contextRequirements [{requirement {
         class=FortControllerRequirement_HasCID
         properties {allowedCIds [/BRCosmetics/Athena/Items/Cosmetics/Characters/CID_A_424_Athena_Commando_M_Lancelot]}
        }}]
      priority {value=0, curve {curveTable=null, rowName=None}}
     }
     {
      message="Once again, I am unable to access the dark web. Stop asking."
      contextRequirements [{requirement {
         class=FortControllerRequirement_HasCID
         properties {allowedCIds [/BRCosmetics/Athena/Items/Cosmetics/Characters/CID_A_428_Athena_Commando_F_PinkWidow]}
        }}]
      priority {value=0, curve {curveTable=null, rowName=None}}
     }
     {
      message=Are you metal too?
      contextRequirements [{requirement {
         class=FortControllerRequirement_HasCID
         properties {allowedCIds [/BRCosmetics/Athena/Items/Cosmetics/Characters/CID_A_427_Athena_Commando_F_Fuchsia]}
        }}]
      priority {value=0, curve {curveTable=null, rowName=None}}
     }
     {
      message=Teach me your customization protocol.
      contextRequirements [{requirement {
         class=FortControllerRequirement_HasCID
         properties {allowedCIds [/BRCosmetics/Athena/Items/Cosmetics/Characters/CID_A_429_Athena_Commando_M_Collectable]}
        }}]
      priority {value=0, curve {curveTable=null, rowName=None}}
     }
     {
      message="The Syndicate has no use for an old bot like me. "
      contextRequirements [{requirement {
         class=FortControllerRequirement_HasCID
         properties {allowedCIds [&CID_A_425_Athena_Commando_F_BlueJay]}
        }}]
      priority {value=0, curve {curveTable=null, rowName=None}}
     }
     {
      message=Begin protocol: high-stakes adventure.
      contextRequirements [{requirement {
         class=FortControllerRequirement_HasCID
         properties {allowedCIds [/BRCosmetics/Athena/Items/Cosmetics/Characters/CID_A_423_Athena_Commando_M_Canary]}
        }}]
      priority {value=0, curve {curveTable=null, rowName=None}}
     }
     {
      message=I sense a robo-darkness within you...
      contextRequirements [{requirement {
         class=FortControllerRequirement_HasCID
         properties {allowedCIds [&CID_A_422_Athena_Commando_M_Realm]}
        }}]
      priority {value=0, curve {curveTable=null, rowName=None}}
     }
    ]
   }
  }
 }
 ConversationLinkNode_0 {id=#4, class=ConversationLinkNode, properties {remoteEntryTag {tagName=Conversation.Entry.BasicServices}}}
 ConversationLinkNode_1 {id=#3, class=ConversationLinkNode, properties {remoteEntryTag {tagName=Conversation.Entry.NPCLibrary.Quests}}}
]
